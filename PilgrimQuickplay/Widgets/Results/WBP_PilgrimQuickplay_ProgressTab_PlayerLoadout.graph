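# Player card widget for the Pilgrim Quickplay results Progress tab, showing name, avatar portrait, platform icon, voice status and ready-to-return checkmark.
# SetPlayerState fills the card, async-loads the avatar into a dynamic material, tints bracket and glow by pin color, and reads initial readiness.
# The widget collapses itself for bots or missing avatars.
# Construct subscribes to readiness-change router events, filtered by matching pawn; Destruct unsubscribes and clears references.
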WidgetBlueprint /PilgrimQuickplay/Widgets/Results/WBP_PilgrimQuickplay_ProgressTab_PlayerLoadout
graphs [
 EventGraph {nodes [
   n1 = SetPlayerState!() -> (OutputDelegate, then, PlayerState:Player State Object Reference) {is=title K2Node_CustomEvent}
   > n2 = Sequence!(execute) -> (then_0, then_1, then_2:Exec)
   > n3 = Set PlayerState(PlayerState:Player State Object Reference=@n1.PlayerState) -> (Output_Get:Player State Object Reference)
   > SetPlayerName!(execute, self, PlayerState:Actor Object Reference=@n3) -> (then)
   > n4 = Cast To FortPlayerStateAthena!(execute, Object=@n1.PlayerState) -> (then, CastFailed, AsFort Player State Athena:Fort Player State Athena Object Reference) {is=title K2Node_DynamicCast}
   > n5 = Set As Fort Player State Athena(As Fort Player State Athena:Fort Player State Athena Object Reference=@n4) -> (Output_Get:Fort Player State Athena Object Reference)
   > n6 = Branch(Condition=NOT Boolean(A=Get bIsABot!(self:Player State Object Reference=@n5) -> (bIsABot:bool)))
   n7 = GetSocialAvatarBrush!(self:Fort Player State Zone Object Reference=@n5, bSmallImage:bool=false) -> (ReturnValue:Texture 2D Soft Object Reference)
   > n8 = Branch(Condition=IsValidSoftObjectReference(SoftObjectReference=@n7))
   > n9 = Async Load Asset(Asset=@n7)
   n10 = Cast To Texture2D(execute=@n9.Completed, Object=@n9)
   > SetTextureParameterValue(
    self=Get PortraitDynamicMaterial() -> (PortraitDynamicMaterial:Material Instance Dynamic Object Reference)
    ParameterName=Portrait
    Value=@n10
   )
   SetVisibility(execute=[@n8.else, @n6.else], InVisibility=Collapsed)
   n11 = GetDynamicMaterial(execute=@n2.then_1, self=Get Color_Bracket() -> (Color_Bracket:Image Object Reference))
   n12 = GetPinColor!(self:Fort Player State Zone Object Reference=Get As Fort Player State Athena() -> (As Fort Player State Athena:Fort Player State Athena Object Reference)) -> (ReturnValue:Linear Color Structure)
   > SetVectorParameterValue(self=@n11, ParameterName=ColorA, Value=@n12)
   > SetVectorParameterValue(self=@n11, ParameterName=ColorB, Value=@n12)
   > SetColorAndOpacity(self:Image Object Reference=Get BracketGlow() -> (BracketGlow:Image Object Reference), InColorAndOpacity:Linear Color Structure=@n12)
   n13 = Get As Fort Player State Athena()
   SetPlatformPrefixIcon!(
    execute=@n2.then_2
    Image:Image Object Reference=Get Image_PlatformIcon() -> (Image_PlatformIcon:Image Object Reference)
    OtherPlayerPlatform:String=GetPlatform!(self:Fort Player State Object Reference=@n13) -> (ReturnValue:String)
   ) -> (then)
   > HookUpVoiceIcon!(
    execute
    self:Pilgrim Player Info Widget Base Object Reference
    Icon:Fort Voice Chat Status Icon Object Reference=Get FortVoiceChatStatusIcon() -> (FortVoiceChatStatusIcon:Fort Voice Chat Status Icon Proxy Anim Object Reference)
    Player:Fort Player State Athena Object Reference=@n13
   ) -> (then)
   > n14 = GetPilgrimCorePlayerStateComponent!(execute, PlayerState:Player State Object Reference=Get PlayerState() -> (PlayerState:Player State Object Reference)) -> (then, ReturnValue:Pilgrim Core Player State Component Object Reference)
   > RefreshReadyIndicator!(execute, self, isReady:bool=GetIsReadyToReturn!(self:Pilgrim Core Player State Component Object Reference=@n14) -> (ReturnValue:bool)) -> (then)
   Event Construct()
   > Parent: Construct()
   > n15 = GetDynamicMaterial(self=Get Image_SquadMemberPortrait() -> (Image_SquadMemberPortrait:Fort Lazy Image Object Reference))
   > Set PortraitDynamicMaterial(PortraitDynamicMaterial:Material Instance Dynamic Object Reference=@n15) -> (Output_Get:Material Instance Dynamic Object Reference)
   > n16 = PilgrimWaitForEventRouter!(execute:Exec, ContextActor:Actor Object Reference=GetOwningPlayer()) -> (then:Exec, OnEventRouterAvailable:Exec, EventRouter:Gameplay Event Router Component Object Reference) {is=title K2Node_AsyncAction}
   n17 = Set Event Router(execute=@n16.OnEventRouterAvailable, Event Router:Gameplay Event Router Component Object Reference=@n16) -> (Output_Get:Gameplay Event Router Component Object Reference)
   > n18 = StartListeningToEvent!(
    execute:Exec
    Target:Gameplay Event Router Component Object Reference=@n17
    EventType:Script Struct Object Reference=/Script/PilgrimCoreRuntime.PilgrimEvent_PlayerState_ReadinessToReturnChanged
    Context:Object Reference
    EventBubblingRule:EEventBubblingRule Enum
   ) -> (
    then:Exec
    AsyncTaskProxy:Async Action Start Listening To Event Object Reference
    OnEventReceived:Exec
    Payload:Pilgrim Event Player State Readiness To Return Changed Structure
   )
   > Set ReadinessToReturnAsyncActionRef(ReadinessToReturnAsyncActionRef:Async Action Start Listening To Event Object Reference=@n18) -> (Output_Get:Async Action Start Listening To Event Object Reference)
   n19 = Get Fort Pawn From Object!(execute=@n18.OnEventReceived, InObject:Object Reference=Get PlayerState()) -> (Valid:Exec, NotValid:Exec, FortPawn:Fort Pawn Object Reference) {is=title K2Node_CallFunction}
   n20 = Break Pilgrim Event Player State Readiness to Return Changed!(PilgrimEvent_PlayerState_ReadinessToReturnChanged:"Pilgrim Event Player State Readiness To Return Changed Structure (by ref)"=@n18.Payload) -> (ForPlayer:Fort Player Pawn Object Reference, bIsReadyToReturn:bool=~) {is=title K2Node_BreakStruct}
   > Branch(Condition="Equal (Object)"(A=@n19, B=@n20))
   > RefreshReadyIndicator(isReady=@n20.bIsReadyToReturn)
   Event Destruct()
   > Parent: Destruct()
   > StopListeningToEvent(self:Async Action Start Listening To Event Object Reference=Get ReadinessToReturnAsyncActionRef() -> (ReadinessToReturnAsyncActionRef:Async Action Start Listening To Event Object Reference))
   > Set ReadinessToReturnAsyncActionRef()
   > Set Event Router()
   > Set PlayerState()
   > Set As Fort Player State Athena()
  ]}
 RefreshReadyIndicator {nodes [
   n1 = RefreshReadyIndicator!() -> (then, isReady:bool=~) {is=title K2Node_FunctionEntry}
   > SetVisibility(
    self=Get ReadyCheckmark() -> (ReadyCheckmark:Overlay Object Reference)
    InVisibility=Select(Option 0=Hidden, Option 1=SelfHitTestInvisible, Index=@n1)
   )
  ]}
 SetPlayerName {nodes [
   n1 = SetPlayerName!() -> (then, PlayerState:Actor Object Reference) {is=title K2Node_FunctionEntry}
   > Is Valid(InputObject=@n1)
   > "SetText (Text)"(
    self=Get Text_PlayerName() -> (Text_PlayerName:Common Text Object Reference)
    InText="To Text (String)"(InString=GetPlayerNameSafe!(Actor:Actor Object Reference=@n1) -> (ReturnValue:String))
   )
  ]}
]
variables [
 As Fort Player State Athena {type=FortPlayerStateAthena, default=null}
 Event Router {type=GameplayEventRouterComponent, default=null}
 PlayerState {type=PlayerState, default=null}
 PortraitDynamicMaterial {type=MaterialInstanceDynamic}
 ReadinessToReturnAsyncActionRef {type=AsyncAction_StartListeningToEvent, default=null}
]
widgets {
 info {inheritedWidgetCount=0, parentClass=/Script/PilgrimCoreUI.PilgrimPlayerInfoWidgetBase, rootWidgetClass=&Overlay, widgetCount=19}
 widgets [
  Overlay_PlayerLoadout {widgetClassPath=&Overlay, widget=Overlay_PlayerLoadout}
  PlayerLoadout1 {widgetClassPath=&VerticalBox, parent=Overlay_PlayerLoadout, widget=PlayerLoadout1, slot {class=OverlaySlot}}
  PlayerNamePlatformSpeaker {
   widgetClassPath=&HorizontalBox
   parent=PlayerLoadout1
   widget=PlayerNamePlatformSpeaker
   properties {flowDirectionPreference=Culture}
   slot {class=VerticalBoxSlot, properties {horizontalAlignment=HAlign_Center, verticalAlignment=VAlign_Center}}
  }
  VerticalBox_0 {
   widgetClassPath=&VerticalBox
   parent=PlayerNamePlatformSpeaker
   widget=VerticalBox_0
   slot {class=HorizontalBoxSlot, properties {horizontalAlignment=HAlign_Center, verticalAlignment=VAlign_Center, padding=<ltrb 0 -5 0 -5>}}
  }
  HorizontalBox_0 {
   widgetClassPath=&HorizontalBox
   parent=VerticalBox_0
   widget=HorizontalBox_0
   properties {renderTransform {angle=0, scale=<1 1>, shear=<0 0>, translation=<5 0>}}
   slot {class=VerticalBoxSlot, properties {horizontalAlignment=HAlign_Center}}
  }
  SizeBox_0 {
   widgetClassPath=&SizeBox
   parent=HorizontalBox_0
   widget=SizeBox_0
   properties {bOverride_HeightOverride=true, bOverride_WidthOverride=true, heightOverride=22, widthOverride=22}
   slot {class=HorizontalBoxSlot, properties {horizontalAlignment=HAlign_Center, verticalAlignment=VAlign_Center}}
  }
  Image_PlatformIcon {
   widgetClassPath=&Image
   bIsVariable=true
   parent=SizeBox_0
   widget=Image_PlatformIcon
   properties {
    visibility=Collapsed
    brush {
     bIsDynamicallyLoaded=false
     drawAs=Image
     imageType=NoImage
     mirroring=NoMirror
     resourceName=None
     resourceObject=/Game/UI/Friends_UI/Social/Console_PlatformIcon_32x
     tiling=NoTile
     imageSize=<32 32>
     margin=<ltrb 0 0 0 0>
     outlineSettings {
      bUseBrushTransparency=false
      roundingType=HalfHeightRadius
      width=0
      color {colorUseRule=UseColor_Specified, specifiedColor=<rgba 0 0 0 0>}
      cornerRadii=<0 0 0 0>
     }
     tintColor {colorUseRule=UseColor_Specified, specifiedColor=<rgba 1 1 1 1>}
     uVRegion {bIsValid=false, max=<0 0>, min=<0 0>}
    }
   }
   slot {class=SizeBoxSlot}
  }
  FortVoiceChatStatusIcon {
   widgetClassPath=/Script/FortniteUI.FortVoiceChatStatusIconProxyAnim
   bIsVariable=true
   parent=HorizontalBox_0
   widget=FortVoiceChatStatusIcon
   slot {class=HorizontalBoxSlot}
  }
  Text_PlayerName {
   widgetClassPath=&CommonTextBlock
   bIsVariable=true
   parent=VerticalBox_0
   widget=Text_PlayerName
   properties {
    text=TestPlayerName
    font {
     bForceMonospaced=false
     bMaterialIsStencil=false
     fontMaterial=null
     fontObject=&HeadingNow
     letterSpacing=0
     monospacedWidth=1
     size=13.5
     skewAmount=0
     typefaceFontName=75 Medium
     outlineSettings {
      bApplyOutlineToDropShadows=true
      bMiteredCorners=false
      bSeparateFillAlpha=false
      outlineMaterial=null
      outlineSize=0
      outlineColor=<rgba 0 0 0 1>
     }
    }
    margin=<ltrb 0 0 2 0>
   }
   slot {class=VerticalBoxSlot}
  }
  Overlay_PlayerBox {
   widgetClassPath=&Overlay
   parent=PlayerLoadout1
   widget=Overlay_PlayerBox
   slot {class=VerticalBoxSlot, properties {horizontalAlignment=HAlign_Center, padding=<ltrb 0 4 0 0>}}
  }
  Spacer_101 {
   widgetClassPath=&Spacer
   parent=Overlay_PlayerBox
   widget=Spacer_101
   properties {size=<256 256>}
   slot {class=OverlaySlot, properties {horizontalAlignment=HAlign_Center, verticalAlignment=VAlign_Center}}
  }
  WBP_Sparks_BackingPanel_C_0 {
   widgetClassPath=/PilgrimCore/Widgets/SparksUILibrary/WBP_Sparks_BackingPanel
   bIsVariable=true
   parent=Overlay_PlayerBox
   widget=WBP_Sparks_BackingPanel_C_0
   properties {blurFrame=false, cornerRadius=Small, darkenLayer=false, eSparksBackingPanelType=Vertical Gradient, renderOpacity=0.75}
   slot {class=OverlaySlot, properties {horizontalAlignment=HAlign_Fill, verticalAlignment=VAlign_Fill}}
  }
  Image_SquadMemberPortrait {
   widgetClassPath=&FortLazyImage
   bIsVariable=true
   parent=Overlay_PlayerBox
   widget=Image_SquadMemberPortrait
   properties {brush {
     bIsDynamicallyLoaded=false
     drawAs=Image
     imageType=NoImage
     mirroring=NoMirror
     resourceName=None
     resourceObject=/PilgrimCore/Widgets/Materials/MI_UI_PIligrim_Portait_RoundCorners
     tiling=NoTile
     imageSize=<256 256>
     margin=<ltrb 0 0 0 0>
     outlineSettings {
      bUseBrushTransparency=false
      roundingType=HalfHeightRadius
      width=0
      color {colorUseRule=UseColor_Specified, specifiedColor=<rgba 0 0 0 0>}
      cornerRadii=<0 0 0 0>
     }
     tintColor {colorUseRule=UseColor_Specified, specifiedColor=<rgba 1 1 1 1>}
     uVRegion {bIsValid=false, max=<0 0>, min=<0 0>}
    }}
   slot {class=OverlaySlot, properties {horizontalAlignment=HAlign_Center, verticalAlignment=VAlign_Bottom}}
  }
  Color_Bracket {
   widgetClassPath=&Image
   bIsVariable=true
   parent=Overlay_PlayerBox
   widget=Color_Bracket
   properties {
    visibility=SelfHitTestInvisible
    brush {
     bIsDynamicallyLoaded=false
     drawAs=Image
     imageType=NoImage
     mirroring=NoMirror
     resourceName=None
     resourceObject=/PilgrimCore/Widgets/Materials/MI_Piligrim_Bandmate_PostmatchBracket_ProgressScreen
     tiling=NoTile
     imageSize=<64 64>
     margin=<ltrb 0 0 0 0>
     outlineSettings {
      bUseBrushTransparency=false
      roundingType=HalfHeightRadius
      width=0
      color {colorUseRule=UseColor_Specified, specifiedColor=<rgba 0 0 0 0>}
      cornerRadii=<0 0 0 0>
     }
     tintColor {colorUseRule=UseColor_Specified, specifiedColor=<rgba 1 1 1 1>}
     uVRegion {bIsValid=false, max=<0 0>, min=<0 0>}
    }
    renderTransform {angle=0, scale=<1 1>, shear=<0 0>, translation=<0 2>}
    renderTransformPivot=<0 1>
   }
   slot {class=OverlaySlot, properties {horizontalAlignment=HAlign_Fill, verticalAlignment=VAlign_Fill}}
  }
  BracketGlow {
   widgetClassPath=&Image
   bIsVariable=true
   parent=Overlay_PlayerBox
   widget=BracketGlow
   properties {
    visibility=SelfHitTestInvisible
    brush {
     bIsDynamicallyLoaded=false
     drawAs=Box
     imageType=NoImage
     mirroring=NoMirror
     resourceName=None
     resourceObject=/PilgrimCore/Widgets/Materials/MI_UI_Pilgrim_ProgressPlayerPortrait_Glow
     tiling=NoTile
     imageSize=<64 64>
     margin=<ltrb 0.5 0.5 0.5 0.5>
     outlineSettings {
      bUseBrushTransparency=false
      roundingType=HalfHeightRadius
      width=0
      color {colorUseRule=UseColor_Specified, specifiedColor=<rgba 0 0 0 0>}
      cornerRadii=<0 0 0 0>
     }
     tintColor {colorUseRule=UseColor_Specified, specifiedColor=<rgba 1 1 1 1>}
     uVRegion {bIsValid=false, max=<0 0>, min=<0 0>}
    }
    renderTransform {angle=0, scale=<1.175 1>, shear=<0 0>, translation=<0 26>}
   }
   slot {class=OverlaySlot, properties {horizontalAlignment=HAlign_Fill, verticalAlignment=VAlign_Bottom}}
  }
  ReadyCheckmark {
   widgetClassPath=&Overlay
   bIsVariable=true
   parent=Overlay_PlayerBox
   widget=ReadyCheckmark
   properties {visibility=Hidden, renderTransform {angle=0, scale=<1 1>, shear=<0 0>, translation=<0 4>}}
   slot {class=OverlaySlot, properties {horizontalAlignment=HAlign_Center, verticalAlignment=VAlign_Bottom}}
  }
  Check_Glow {
   widgetClassPath=&Image
   parent=ReadyCheckmark
   widget=Check_Glow
   properties {
    visibility=SelfHitTestInvisible
    brush {
     bIsDynamicallyLoaded=false
     drawAs=Image
     imageType=NoImage
     mirroring=NoMirror
     resourceName=None
     resourceObject=/PilgrimCore/Widgets/Materials/MI_UI_PilgrimReadyGlow
     tiling=NoTile
     imageSize=<24 24>
     margin=<ltrb 0 0 0 0>
     outlineSettings {
      bUseBrushTransparency=false
      roundingType=HalfHeightRadius
      width=0
      color {colorUseRule=UseColor_Specified, specifiedColor=<rgba 0 0 0 0>}
      cornerRadii=<0 0 0 0>
     }
     tintColor {colorUseRule=UseColor_Specified, specifiedColor=<rgba 1 1 1 1>}
     uVRegion {bIsValid=false, max=<0 0>, min=<0 0>}
    }
    renderTransform {angle=0, scale=<2.5 2.5>, shear=<0 0>, translation=<0 0>}
   }
   slot {class=OverlaySlot, properties {horizontalAlignment=HAlign_Fill, verticalAlignment=VAlign_Fill}}
  }
  Check_Pulse {
   widgetClassPath=&Image
   parent=ReadyCheckmark
   widget=Check_Pulse
   properties {
    visibility=SelfHitTestInvisible
    brush {
     bIsDynamicallyLoaded=false
     drawAs=Image
     imageType=NoImage
     mirroring=NoMirror
     resourceName=None
     resourceObject=/PilgrimCore/Widgets/Materials/MI_UI_PilgrimReadyStatePulse
     tiling=NoTile
     imageSize=<24 24>
     margin=<ltrb 0 0 0 0>
     outlineSettings {
      bUseBrushTransparency=false
      roundingType=HalfHeightRadius
      width=0
      color {colorUseRule=UseColor_Specified, specifiedColor=<rgba 0 0 0 0>}
      cornerRadii=<0 0 0 0>
     }
     tintColor {colorUseRule=UseColor_Specified, specifiedColor=<rgba 1 1 1 1>}
     uVRegion {bIsValid=false, max=<0 0>, min=<0 0>}
    }
    renderTransform {angle=0, scale=<2.5 2.5>, shear=<0 0>, translation=<0 0>}
   }
   slot {class=OverlaySlot, properties {horizontalAlignment=HAlign_Fill, verticalAlignment=VAlign_Fill}}
  }
  StatusIndicator {
   widgetClassPath=/FortUILibrary/UIKit/StatusIndicator/WBP_UIKit_StatusIndicator
   bIsVariable=true
   parent=ReadyCheckmark
   widget=StatusIndicator
   properties {icon=/FortUILibrary/Textures/MaterialInstances/MI_UIKit_Icon_Check, iconSize=Compact, overrideStatusIndicatorVisibility=true}
   slot {class=OverlaySlot}
  }
 ]
}
