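# Base Widget Blueprint for Pilgrim Quickplay results UI, derived from the native PilgrimQuickplay_UIResultBase class.
# It has no widget tree and an empty event graph, so it adds no visuals or Blueprint logic of its own.
# It declares a music battle instance reference, handle arrays for async, gameplay and stateful event listeners, and flags for wanting local or replicated score updates.
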WidgetBlueprint /PilgrimQuickplay/Widgets/Results/WBP_Pilgrim_UI_Results_Base
graphs [EventGraph {links [], nodes []}]
variables [
 AsyncEventHandles {type="array<CancellableAsyncAction>", default []}
 GameplayEventHandles {type="array<AsyncAction_StartListeningToEvent>", default []}
 MusicBattleInstance {type=MusicBattleInstance_C, default=null}
 StatefulGameplayEventHandles {type="array<AsyncAction_StartListeningToStatefulEvent>", default []}
 WantsLocalPlayerScoreUpdates {type=boolean}
 WantsReplicatedScoreUpdates {type=boolean}
]
widgets {
 info {inheritedWidgetCount=0, parentClass=/Script/PilgrimQuickplayUI.PilgrimQuickplay_UIResultBase, rootWidgetClass=null, widgetCount=0}
 widgets []
}
