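# Pilgrim Quickplay results recap widget, derived from WBP_Pilgrim_UI_Results_Base, that shows a "QUEST PROGRESS" panel.
# A bordered vertical box holds the title and a WBP_PilgrimQuestList set to hide completed and unchanged quests.
# On initialize it caches the owning player's Quickplay player state component, though the event graph does not use it further.
# Losing focus prints a debug string; the song result, event router and index variables are unused in these graphs.
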WidgetBlueprint /PilgrimQuickplay/Widgets/Results/WBP_PilgrimQuickplay_Results_RecapDisplay
graphs [
 EventGraph {nodes [
   n1 = Event PreConstruct()
   > Parent: PreConstruct(IsDesignTime=@n1.IsDesignTime)
   Event OnInitialized()
   > Initialize()
   Event OnFocusLost!() -> (OutputDelegate, then, InFocusEvent:Focus Event Structure) {is=title K2Node_Event}
   n2 = Self-Reference()
   > PrintString!(
    execute
    InString=Append(A="On Focus Lost: ", B=GetDisplayName(Object=@n2))
    bPrintToScreen
    bPrintToLog
    TextColor=(R=1,G=0,B=0.9676,A=1)
    Duration
    Key
   ) -> (then)
  ]}
 GetPlayerStateComponent {nodes [
   GetPlayerStateComponent() {is=title K2Node_FunctionEntry}
   > n1 = GetPilgrimQuickplayPlayerStateComponent!(
    execute
    PlayerState:Player State Object Reference=Get PlayerState!(self:Controller Object Reference=GetOwningPlayer()) -> (PlayerState:Player State Object Reference)
   ) -> (then, ReturnValue:Pilgrim Quickplay Player State Component Object Reference)
   > Return Node!(execute, PQPStateComp:Pilgrim Quickplay Player State Component Object Reference=@n1)
  ]}
 Initialize {nodes [
   Initialize() {is=title K2Node_FunctionEntry}
   > n1 = GetPlayerStateComponent!(execute, self) -> (then, PQPStateComp:Pilgrim Quickplay Player State Component Object Reference) {is=title K2Node_CallFunction}
   > Set Local PQPState Comp(Local PQPState Comp:Pilgrim Quickplay Player State Component Object Reference=@n1) -> (Output_Get:Pilgrim Quickplay Player State Component Object Reference)
  ]}
]
variables [
 Event Router {type=GameplayEventRouterComponent, default=null}
 Local PQPState Comp {type=PilgrimQuickplayPlayerStateComponent, default=null}
 Local PQPState Index {type=integer}
 SongResultData {
  type=PilgrimQuickplaySongResultData
  default {
   localPlayerBandDataIndex=-1
   localPlayerDataIndex=-1
   songName=""
   songShortName=None
   team=0
   bandData []
   playerData []
   songTags {gameplayTags []}
  }
 }
]
widgets {
 info {
  inheritedWidgetCount=0
  parentClass=&/WBP_Pilgrim_UI_Results_Base.WBP_Pilgrim_UI_Results_Base_C
  rootWidgetClass=&Overlay
  widgetCount=5
 }
 widgets [
  Overlay_8 {widgetClassPath=&Overlay, widget=Overlay_8}
  CommonBorder_1 {
   widgetClassPath=&CommonBorder
   parent=Overlay_8
   widget=CommonBorder_1
   properties {style=/FortUILibrary/Styles/Border/Border-MainL-Black.Border-MainL-Black_C, padding=<ltrb 36 30 36 24>}
   slot {class=OverlaySlot, properties {horizontalAlignment=HAlign_Fill, verticalAlignment=VAlign_Fill}}
  }
  ContentParent {
   widgetClassPath=&VerticalBox
   parent=CommonBorder_1
   widget=ContentParent
   slot {class=BorderSlot, properties {padding=<ltrb 36 30 36 24>}}
  }
  Title {
   widgetClassPath=&CommonTextBlock
   bIsVariable=true
   parent=ContentParent
   widget=Title
   properties {
    style=/FortUILibrary/Styles/Typography/_TextStyle_SubH4._TextStyle_SubH4_C
    text=QUEST PROGRESS
    font {
     bForceMonospaced=false
     bMaterialIsStencil=false
     fontMaterial=null
     fontObject=&BurbankSmall
     letterSpacing=0
     monospacedWidth=1
     size=24
     skewAmount=0
     typefaceFontName=Black
     outlineSettings {
      bApplyOutlineToDropShadows=false
      bMiteredCorners=false
      bSeparateFillAlpha=false
      outlineMaterial=null
      outlineSize=0
      outlineColor=<rgba 0 0 0 1>
     }
    }
    shadowOffset=<0 0>
    strikeBrush {
     bIsDynamicallyLoaded=false
     drawAs=NoDrawType
     imageType=NoImage
     mirroring=NoMirror
     resourceName=None
     resourceObject=null
     tiling=NoTile
     imageSize=<32 32>
     margin=<ltrb 0 0 0 0>
     outlineSettings {
      bUseBrushTransparency=false
      roundingType=HalfHeightRadius
      width=0
      color {colorUseRule=UseColor_Specified, specifiedColor=<rgba 0 0 0 0>}
      cornerRadii=<0 0 0 0>
     }
     tintColor {colorUseRule=UseColor_Specified, specifiedColor=<rgba 1 1 1 1>}
     uVRegion {bIsValid=false, max=<0 0>, min=<0 0>}
    }
   }
   slot {class=VerticalBoxSlot, properties {horizontalAlignment=HAlign_Center, padding=<ltrb 0 0 0 24>}}
  }
  WBP_PilgrimQuestList {
   widgetClassPath=/PilgrimQuickplay/Widgets/Quests/WBP_PilgrimQuestList
   bIsVariable=true
   parent=ContentParent
   widget=WBP_PilgrimQuestList
   properties {bShowCompletedQuests=false, bShowUnchangedQuests=false}
   slot {class=VerticalBoxSlot, properties {horizontalAlignment=HAlign_Center, size {sizeRule=Fill, value=1}}}
  }
 ]
}
