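# Pilgrim Quickplay results widget that shows a player's avatar with a check-mark icon for readiness.
# SetPlayer caches the player state, starts readiness listening and fills the avatar's dynamic material with pin color and async-loaded social avatar.
# An invalid avatar reference is retried on a one-second delay against a countdown.
# EventRouter readiness events matching the cached player toggle the check mark; Destruct stops the listener.
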WidgetBlueprint /PilgrimQuickplay/Widgets/Results/WBP_PilgrimQuickplay_AvatarIcon_Ready
graphs [EventGraph {nodes [
   n1 = SetPlayerImage!() -> (OutputDelegate, then, FortPlayerStateAthena:Fort Player State Athena Object Reference) {is=title K2Node_CustomEvent}
   > n2 = GetDynamicMaterial(self=Get PlayerImage() -> (PlayerImage:Image Object Reference))
   > SetVectorParameterValue(
    self=@n2
    ParameterName=BgColor
    Value=GetPinColor!(self:Fort Player State Zone Object Reference=@n1.FortPlayerStateAthena) -> (ReturnValue:Linear Color Structure)
   )
   > n3 = Set PlayerIconRetryCount(PlayerIconRetryCount:int=Get PlayerIconRetryCount_Default() -> (PlayerIconRetryCount_Default:int=0)) -> (Output_Get:int=0)
   n4 = GetSocialAvatarBrush!(self:Fort Player State Zone Object Reference=@n1.FortPlayerStateAthena, bSmallImage:bool=true) -> (ReturnValue:Texture 2D Soft Object Reference)
   n5 = Branch(execute=[@n9.then, @n3.then], Condition=IsValidSoftObjectReference(SoftObjectReference=@n4))
   > n6 = Async Load Asset(Asset=@n4)
   n7 = Cast To Texture2D(execute=@n6.Completed, Object=@n6)
   > SetTextureParameterValue(self=@n2, ParameterName=PlayerIcon, Value=@n7)
   n8 = Branch(execute=@n5.else, Condition="integer > integer"(A=Get PlayerIconRetryCount() -> (PlayerIconRetryCount:int=0)))
   > Decrement Int(Value=Get PlayerIconRetryCount() -> (PlayerIconRetryCount=0))
   > n9 = Delay(Duration=1)
   PrintString!(execute=@n8.else, InString=Failed to get a vaild texture reference for player, bPrintToScreen, bPrintToLog, TextColor, Duration, Key) -> (then)
   n10 = StartListeningForReadinessChangedEvents!() -> (OutputDelegate, then, FortPlayerStateAthena:Fort Player State Athena Object Reference) {is=title K2Node_CustomEvent}
   > n11 = Is Valid(InputObject=Get EventRouter() -> (EventRouter:Gameplay Event Router Component Object Reference))
   n12 = PilgrimWaitForEventRouter!(execute:Exec=@n11.Is Not Valid, ContextActor:Actor Object Reference=@n10.FortPlayerStateAthena) -> (then:Exec, OnEventRouterAvailable:Exec, EventRouter:Gameplay Event Router Component Object Reference) {is=title K2Node_AsyncAction}
   n13 = Set EventRouter(execute=@n12.OnEventRouterAvailable, EventRouter:Gameplay Event Router Component Object Reference=@n12) -> (Output_Get:Gameplay Event Router Component Object Reference)
   > n14 = StartListeningToEvent!(
    execute:Exec
    Target:Gameplay Event Router Component Object Reference=@n13
    EventType:Script Struct Object Reference=/Script/PilgrimCoreRuntime.PilgrimEvent_PlayerState_ReadinessChanged
    Context:Object Reference
    EventBubblingRule:EEventBubblingRule Enum
    None:%=@n13
   ) -> (
    then:Exec
    AsyncTaskProxy:Async Action Start Listening To Event Object Reference
    OnEventReceived:Exec
    Payload:Pilgrim Event Player State Readiness Changed Structure
   )
   > Set ReadinessChangedEventHandle(ReadinessChangedEventHandle:Async Action Start Listening To Event Object Reference=@n14) -> (Output_Get:Async Action Start Listening To Event Object Reference)
   n15 = Get CachedPlayerState() -> (CachedPlayerState:Fort Player State Athena Object Reference)
   n16 = Is Valid(exec=@n14.OnEventReceived, InputObject=@n15)
   PrintString(execute=@n16.Is Not Valid, InString=No valid player state found when trying to update PlayerReadiness icon)
   n17 = Break Pilgrim Event Player State Readiness Changed!(PilgrimEvent_PlayerState_ReadinessChanged:"Pilgrim Event Player State Readiness Changed Structure (by ref)"=@n14.Payload) -> (ForPlayer:Fort Player Pawn Object Reference, bIsReady:bool=~) {is=title K2Node_BreakStruct}
   Branch(
    execute=@n16.Is Valid
    Condition="Equal (Object)"(A=@n15, B=Get PlayerState!(self:Pawn Object Reference=@n17) -> (PlayerState:Player State Object Reference))
   )
   > SetIsReady!(execute, self, IsReady:bool=@n17.bIsReady) -> (then)
   n18 = SetPlayer!() -> (OutputDelegate, then, FortPlayerStateAthena:Fort Player State Athena Object Reference) {is=title K2Node_CustomEvent}
   > n19 = Set CachedPlayerState(CachedPlayerState:Fort Player State Athena Object Reference=@n18.FortPlayerStateAthena) -> (Output_Get:Fort Player State Athena Object Reference)
   > StartListeningForReadinessChangedEvents!(execute, self, FortPlayerStateAthena:Fort Player State Athena Object Reference=@n19) -> (then) {is=title K2Node_CallFunction}
   > SetPlayerImage!(execute, self, FortPlayerStateAthena:Fort Player State Athena Object Reference=@n19) -> (then) {is=title K2Node_CallFunction}
   n20 = SetIsReady!() -> (OutputDelegate, then, IsReady:bool=~) {is=title K2Node_CustomEvent}
   > SetVisibility(
    self=Get ReadyIcon() -> (ReadyIcon:Image Object Reference)
    InVisibility=Select(Option 0=Hidden, Option 1=SelfHitTestInvisible, Index=@n20.IsReady)
   )
   Event Destruct()
   > Parent: Destruct()
   n21 = Get ReadinessChangedEventHandle() -> (ReadinessChangedEventHandle:Async Action Start Listening To Event Object Reference)
   > Is Valid(InputObject=@n21)
   > StopListeningToEvent(self:Async Action Start Listening To Event Object Reference=@n21)
   > Set ReadinessChangedEventHandle()
  ]}]
variables [
 CachedPlayerState {type=FortPlayerStateAthena, default=null}
 EventRouter {type=GameplayEventRouterComponent, default=null}
 PlayerIconRetryCount {type=integer}
 PlayerIconRetryCount_Default {type=integer, default=10}
 ReadinessChangedEventHandle {type=AsyncAction_StartListeningToEvent, default=null}
]
widgets {
 info {inheritedWidgetCount=0, parentClass=&UserWidget, rootWidgetClass=&Overlay, widgetCount=3}
 widgets [
  Overlay_26 {widgetClassPath=&Overlay, widget=Overlay_26}
  PlayerImage {
   widgetClassPath=&Image
   bIsVariable=true
   parent=Overlay_26
   widget=PlayerImage
   properties {brush {
     bIsDynamicallyLoaded=false
     drawAs=Image
     imageType=FullColor
     mirroring=NoMirror
     resourceName=None
     resourceObject=/SparksUICommon/Materials/MI_UI_BandMember_Avatar
     tiling=NoTile
     imageSize=<64 64>
     margin=<ltrb 0 0 0 0>
     outlineSettings {
      bUseBrushTransparency=false
      roundingType=HalfHeightRadius
      width=0
      color {colorUseRule=UseColor_Specified, specifiedColor=<rgba 0 0 0 0>}
      cornerRadii=<0 0 0 0>
     }
     tintColor {colorUseRule=UseColor_Specified, specifiedColor=<rgba 1 1 1 1>}
     uVRegion {bIsValid=false, max=<0 0>, min=<0 0>}
    }}
   slot {class=OverlaySlot, properties {verticalAlignment=VAlign_Bottom}}
  }
  ReadyIcon {
   widgetClassPath=&Image
   bIsVariable=true
   parent=Overlay_26
   widget=ReadyIcon
   properties {brush {
     bIsDynamicallyLoaded=false
     drawAs=Image
     imageType=NoImage
     mirroring=NoMirror
     resourceName=None
     resourceObject=/FortUILibrary/Styles/Border/Materials/M_CircledCheckMark
     tiling=NoTile
     imageSize=<29 29>
     margin=<ltrb 0 0 0 0>
     outlineSettings {
      bUseBrushTransparency=false
      roundingType=HalfHeightRadius
      width=0
      color {colorUseRule=UseColor_Specified, specifiedColor=<rgba 0 0 0 0>}
      cornerRadii=<0 0 0 0>
     }
     tintColor {colorUseRule=UseColor_Specified, specifiedColor=<rgba 1 1 1 1>}
     uVRegion {bIsValid=false, max=<0 0>, min=<0 0>}
    }}
   slot {class=OverlaySlot, properties {verticalAlignment=VAlign_Bottom}}
  }
 ]
}
