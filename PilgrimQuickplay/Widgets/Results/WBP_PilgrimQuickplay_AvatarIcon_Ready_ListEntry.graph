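# Widget Blueprint list entry for the Pilgrim Quickplay results screen, wrapping the avatar icon with ready state for use in a list view.
# It adds no widgets of its own, inheriting from WBP_PilgrimQuickplay_AvatarIcon_Ready.
# When its list item is set, it casts the item to BP_AvatarIcon_Ready_EntryData and passes that data's player state and ready flag to the parent's SetPlayer and SetIsReady.
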WidgetBlueprint /PilgrimQuickplay/Widgets/Results/WBP_PilgrimQuickplay_AvatarIcon_Ready_ListEntry
graphs [EventGraph {nodes [
   n1 = Event OnListItemObjectSet()
   > n2 = Cast To BP_AvatarIcon_Ready_EntryData!(execute, Object=@n1.ListItemObject) -> (then, CastFailed, AsBP Avatar Icon Ready Entry Data:BP Avatar Icon Ready Entry Data Object Reference) {is=title K2Node_DynamicCast}
   > n3 = Sequence()
   n4 = Get!(self:BP Avatar Icon Ready Entry Data Object Reference=@n2) -> (
    EntryData_PlayerState_4_#1:Fort Player State Athena Object Reference
    EntryData_IsReady_2_#2:bool=~
   )
   > Parent: SetPlayer!(execute, FortPlayerStateAthena:Fort Player State Athena Object Reference=@n4) -> (then) {is=title K2Node_CallParentFunction}
   Parent: SetIsReady!(execute=@n3.then_1, IsReady:bool=@n4.EntryData_IsReady_2_#2) -> (then) {is=title K2Node_CallParentFunction}
   Event PreConstruct()
   Event Construct()
   Event Tick()
  ]}]
variables [ListItemObject {type=Object, default=null}]
widgets {
 info {
  inheritedWidgetCount=0
  parentClass=&/WBP_PilgrimQuickplay_AvatarIcon_Ready.WBP_PilgrimQuickplay_AvatarIcon_Ready_C
  rootWidgetClass=null
  widgetCount=0
 }
 widgets []
}
