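# Pilgrim Quickplay results widget that shows a song's difficulty as a one-letter badge.
# SetDifficulty maps the EPilgrimSongDifficulty enum to E, M, H, X or N through a Select and passes it to SetDifficultyText, which sets the text block.
# UpdateDifficultyState flips the root widget switcher between the played badge with the letter and a dimmed, empty badge for unplayed songs.
# Both badges use the shared difficulty border material at the same fixed size.
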WidgetBlueprint /PilgrimQuickplay/Widgets/Results/WBP_Pilgrim_Results_Difficulty
graphs [
 EventGraph {links [], nodes []}
 SetDifficulty {nodes [
   n1 = SetDifficulty!() -> (then, Difficulty:EPilgrimSongDifficulty Enum=DifficultyEasy) {is=title K2Node_FunctionEntry}
   > SetDifficultyText!(
    execute
    self
    InText:Text=Select!(
     DifficultyEasy:"Text (by ref)"=MakeLiteralText(Value="LOCTEXT(\"E\")")
     DifficultyMedium:"Text (by ref)"=MakeLiteralText(Value="LOCTEXT(\"M\")")
     DifficultyHard:"Text (by ref)"=MakeLiteralText(Value="LOCTEXT(\"H\")")
     DifficultyExpert:"Text (by ref)"=MakeLiteralText(Value="LOCTEXT(\"X\")")
     None:"Text (by ref)"=MakeLiteralText(Value="LOCTEXT(\"N\")")
     Index:EPilgrimSongDifficulty Enum=@n1
    ) -> (ReturnValue:"Text (by ref)")
   ) -> (then)
  ]}
 SetDifficultyText {nodes [
   n1 = SetDifficultyText!() -> (then, InText:Text) {is=title K2Node_FunctionEntry}
   > "SetText (Text)"(self=Get CommonTextBlock-Difficulty() -> (CommonTextBlock-Difficulty:Common Text Object Reference), InText=@n1)
  ]}
 UpdateDifficultyState {nodes [
   n1 = UpdateDifficultyState!() -> (then, SongIsUnplayed:bool=~) {is=title K2Node_FunctionEntry}
   > n2 = Branch(Condition=@n1)
   n3 = Get DifficultyWidgetSwitcher() -> (DifficultyWidgetSwitcher:Common Activatable Widget Switcher Object Reference)
   > SetActiveWidgetIndex(self=@n3, Index=1)
   SetActiveWidgetIndex(execute=@n2.else, self=@n3, Index=0)
  ]}
]
variables []
widgets {
 info {inheritedWidgetCount=0, parentClass=&UserWidget, rootWidgetClass=&CommonActivatableWidgetSwitcher, widgetCount=7}
 widgets [
  DifficultyWidgetSwitcher {widgetClassPath=&CommonActivatableWidgetSwitcher, bIsVariable=true, widget=DifficultyWidgetSwitcher}
  PlayedDifficultyBorder {
   widgetClassPath=&Border
   parent=DifficultyWidgetSwitcher
   widget=PlayedDifficultyBorder
   properties {
    visibility=HitTestInvisible
    background {
     bIsDynamicallyLoaded=false
     drawAs=Image
     imageType=NoImage
     mirroring=NoMirror
     resourceName=None
     resourceObject=/PilgrimCore/Widgets/Materials/M_UI_Pilgrim_Difficulty_Border
     tiling=NoTile
     imageSize=<32 32>
     margin=<ltrb 0 0 0 0>
     outlineSettings {
      bUseBrushTransparency=false
      roundingType=HalfHeightRadius
      width=0
      color {colorUseRule=UseColor_Specified, specifiedColor=<rgba 0 0 0 0>}
      cornerRadii=<0 0 0 0>
     }
     tintColor {colorUseRule=UseColor_Specified, specifiedColor=<rgba 1 1 1 1>}
     uVRegion {bIsValid=false, max=<0 0>, min=<0 0>}
    }
    padding=<ltrb 8 0 8 0>
   }
   slot {class=WidgetSwitcherSlot}
  }
  SizeBox_0 {
   widgetClassPath=&SizeBox
   parent=PlayedDifficultyBorder
   widget=SizeBox_0
   properties {bOverride_HeightOverride=true, bOverride_WidthOverride=true, heightOverride=40, widthOverride=30}
   slot {class=BorderSlot, properties {padding=<ltrb 8 0 8 0>}}
  }
  ScaleBox_0 {
   widgetClassPath=&ScaleBox
   parent=SizeBox_0
   widget=ScaleBox_0
   slot {class=SizeBoxSlot, properties {horizontalAlignment=HAlign_Center, verticalAlignment=VAlign_Center}}
  }
  CommonTextBlock-Difficulty {
   widgetClassPath=&CommonTextBlock
   bIsVariable=true
   parent=ScaleBox_0
   widget=CommonTextBlock-Difficulty
   properties {
    text=M
    colorAndOpacity {colorUseRule=UseColor_Specified, specifiedColor=<rgba 0.5 0.8333 1 1>}
    font {
     bForceMonospaced=false
     bMaterialIsStencil=false
     fontMaterial=null
     fontObject=&HeadingNow
     letterSpacing=0
     monospacedWidth=1
     size=24
     skewAmount=0
     typefaceFontName=75 Medium
     outlineSettings {
      bApplyOutlineToDropShadows=false
      bMiteredCorners=false
      bSeparateFillAlpha=true
      outlineMaterial=&M-FontShadow
      outlineSize=0
      outlineColor=<rgba 0.0833 0 1 1>
     }
    }
    shadowOffset=<0 0>
   }
   slot {class=ScaleBoxSlot}
  }
  UnplayedDifficultyBorder {
   widgetClassPath=&Border
   parent=DifficultyWidgetSwitcher
   widget=UnplayedDifficultyBorder
   properties {
    visibility=HitTestInvisible
    background {
     bIsDynamicallyLoaded=false
     drawAs=Image
     imageType=FullColor
     mirroring=NoMirror
     resourceName=None
     resourceObject=&M_UI_Pilgrim_Difficulty_Border
     tiling=NoTile
     imageSize=<32 32>
     margin=<ltrb 0 0 0 0>
     outlineSettings {
      bUseBrushTransparency=false
      roundingType=HalfHeightRadius
      width=0
      color {colorUseRule=UseColor_Specified, specifiedColor=<rgba 0 0 0 0>}
      cornerRadii=<0 0 0 0>
     }
     tintColor {colorUseRule=UseColor_Specified, specifiedColor=<rgba 1 1 1 1>}
     uVRegion {bIsValid=false, max=<0 0>, min=<0 0>}
    }
    brushColor=<rgba 0.0008 0 0.005 0.3>
    padding=<ltrb 8 0 8 0>
   }
   slot {class=WidgetSwitcherSlot}
  }
  SizeBox {
   widgetClassPath=&SizeBox
   parent=UnplayedDifficultyBorder
   widget=SizeBox
   properties {bOverride_HeightOverride=true, bOverride_WidthOverride=true, heightOverride=40, widthOverride=30}
   slot {class=BorderSlot, properties {padding=<ltrb 8 0 8 0>}}
  }
 ]
}
