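# Common UI banner for Pilgrim results reward notifications, showing a styled title over layered glow, triangle and particle materials.
# PreConstruct sets the text and looks up NotificationType in NotificationStyles to apply soft-reference materials plus text outline and shadow colors.
# Font size and banner height overrides apply on Construct and in design-time preview; Set Notification Text updates the title.
# Intro and outro animations are defined, but this graph never plays them.
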WidgetBlueprint /PilgrimQuickplay/Widgets/Results/WBP_Pilgrim_Results_RewardNotif
animations [
 anim_Intro {
  binding_count=4
  length_seconds=3
  object=&self:anim_Intro
  sequence {bindings [
    Image_TitleBG {
     id=#1
     possessed_class=Image
     tracks [
      {
       type=MovieScene2DTransformTrack
       sections [{
         type=MovieScene2DTransformSection
         channels [
          Scale.X {type=MovieSceneScriptingFloatChannel, keys [[0,1.75],[3,0],[5,2.5],[19,1.75]]}
          Scale.Y {type=MovieSceneScriptingFloatChannel, keys [[0,1],[3,1],[5,1.25],[10,1]]}
         ]
        }]
      }
      {
       type=MovieSceneWidgetMaterialTrack
       sections [{type=MovieSceneParameterSection, channels [Alpha_Final {type=MovieSceneScriptingFloatChannel, keys [[2,0],[6,16],[20,3],[60,3]]}]}]
      }
     ]
    }
    Overlay-AccoladeBG {
     id=#2
     possessed_class=Overlay
     tracks [{
       type=MovieSceneFloatTrack
       display_name=Render Opacity
       sections [{type=MovieSceneFloatSection, channels [None {type=MovieSceneScriptingFloatChannel, keys [[2.481,0],[6,1]]}]}]
      }]
    }
    Text_Title {
     id=#3
     possessed_class=CommonTextBlock
     tracks [
      {
       type=MovieSceneFloatTrack
       display_name=Render Opacity
       sections [{type=MovieSceneFloatSection, channels [None {type=MovieSceneScriptingFloatChannel, keys [[0,0],[3,1]]}]}]
      }
      {
       type=MovieScene2DTransformTrack
       sections [{
         type=MovieScene2DTransformSection
         channels [
          Translation.X {type=MovieSceneScriptingFloatChannel}
          Translation.Y {type=MovieSceneScriptingFloatChannel}
          Angle {type=MovieSceneScriptingFloatChannel}
          Scale.X {type=MovieSceneScriptingFloatChannel, keys [[0,5],[3,0.8],[5,1.05],[9,1]]}
          Scale.Y {type=MovieSceneScriptingFloatChannel, keys [[0,5],[3,0.8],[5,1.05],[9,1]]}
          Shear.X {type=MovieSceneScriptingFloatChannel}
          Shear.Y {type=MovieSceneScriptingFloatChannel}
         ]
        }]
      }
     ]
    }
    Overlay_0 {
     id=#4
     possessed_class=Overlay
     tracks [{
       type=MovieSceneFloatTrack
       display_name=Render Opacity
       sections [{type=MovieSceneFloatSection, end_frame=10, start_frame=0, channels [None {type=MovieSceneScriptingFloatChannel, keys [[0,0],[2,1]]}]}]
      }]
    }
   ]}
 }
 anim_Outro {
  binding_count=1
  length_seconds=0.75
  object=&self:anim_Outro
  sequence {bindings [Overlay_0 {
     id=#5
     possessed_class=Overlay
     tracks [{
       type=MovieSceneFloatTrack
       display_name=Render Opacity
       sections [{type=MovieSceneFloatSection, channels [None {type=MovieSceneScriptingFloatChannel, keys [[0,1],[15,0]]}]}]
      }]
    }]}
 }
]
graphs [
 EventGraph {nodes [
   n1 = Event PreConstruct()
   > Set Text(Text=Get NotificationText() -> (NotificationText:Text), self=Get Text_Title() -> (Text_Title:Common Text Object Reference))
   > SetNotificationStyle()
   > n2 = Branch(Condition=@n1.IsDesignTime)
   SetFontSize!(
    execute=[@n3.then, @n2.then]
    self:Text Object Reference=Get Text_Title()
    DisplayFontSize:float=Get Display Font Size() -> (Display Font Size:float)
   ) -> (then)
   > SetHeightOverride(
    self=Get FortVisualAttachment_SizeBox() -> (FortVisualAttachment_SizeBox:Fort Visual Attachment Object Reference)
    InHeightOverride=Get In Height Override() -> (In Height Override:float)
   )
   n3 = Event Construct()
  ]}
 Set Notification Text {nodes [
   n1 = Set Notification Text!() -> (then, NotificationText:Text) {is=title K2Node_FunctionEntry}
   > n2 = Set NotificationText(NotificationText:Text=@n1) -> (Output_Get:Text)
   > Set Text(Text=@n2, self=Get Text_Title())
  ]}
 SetNotificationStyle {nodes [
   SetNotificationStyle() {is=title K2Node_FunctionEntry}
   n1 = Find(
    TargetMap:Map of E_PilgrimRewardNotifType Enums to F Pilgrim Reward Notif Style Structures=Get NotificationStyles() -> (NotificationStyles:Map of E_PilgrimRewardNotifType Enums to F Pilgrim Reward Notif Style Structures=NewEnumerator0)
    Key:"E_PilgrimRewardNotifType Enum (by ref)"=Get NotificationType() -> (NotificationType:E_PilgrimRewardNotifType Enum=NewEnumerator0)
   ) -> (Value:F Pilgrim Reward Notif Style Structure)
   > Branch(Condition=@n1.ReturnValue)
   n2 = Break F Pilgrim Reward Notif Style!(F_PilgrimRewardNotifStyle:"F Pilgrim Reward Notif Style Structure (by ref)"=@n1) -> (
    BackgroundLayer0_18_#6:Material Interface Soft Object Reference
    BackgroundLayer1_19_#7:Material Interface Soft Object Reference
    BackgroundLayer2_20_#8:Material Interface Soft Object Reference
    Triangles_24_#9:Material Interface Soft Object Reference
    Particles_22_#10:Material Interface Soft Object Reference
    TextOutlineColor_15_#11:Linear Color Structure
    TextShadowColor_16_#12:Linear Color Structure
   ) {is=title K2Node_BreakStruct}
   > SetBrushFromSoftMaterial!(
    execute
    self:Image Object Reference=Get Image_TitleBG() -> (Image_TitleBG:Image Object Reference)
    SoftMaterial:Material Interface Soft Object Reference=@n2
   ) -> (then)
   > SetBrushFromSoftMaterial(
    self=Get Image_TitleBG_1() -> (Image_TitleBG_1:Image Object Reference)
    SoftMaterial=@n2.BackgroundLayer1_19_#7
   )
   > SetBrushFromSoftMaterial(
    self=Get Image_TitleBG_2() -> (Image_TitleBG_2:Image Object Reference)
    SoftMaterial=@n2.BackgroundLayer2_20_#8
   )
   > SetBrushFromSoftMaterial(
    self=[Get ImageTriangleR_1() -> (ImageTriangleR_1:Image Object Reference), Get ImageTriangleL_1() -> (ImageTriangleL_1:Image Object Reference)]
    SoftMaterial=@n2.Triangles_24_#9
   )
   > SetBrushFromSoftMaterial(
    self=[Get ImageParticlesR() -> (ImageParticlesR:Image Object Reference), Get ImageParticlesL() -> (ImageParticlesL:Image Object Reference)]
    SoftMaterial=@n2.Particles_22_#10
   )
   n3 = Get Text_Title()
   > Set Outline Color!(
    In:Exec
    CommonTextWidget:Common Text Object Reference=@n3
    Color:Linear Color Structure=@n2.TextOutlineColor_15_#11
   ) -> (Out:Exec) {is=title K2Node_MacroInstance}
   > SetShadowColorAndOpacity!(
    execute
    self:Text Object Reference=@n3
    InShadowColorAndOpacity:Linear Color Structure=@n2.TextShadowColor_16_#12
   ) -> (then)
  ]}
]
variables [
 Display Font Size {type=number, default=22}
 In Height Override {type=number, default=44}
 NotificationStyles {default {
   Default {
    backgroundLayer0=/PilgrimCore/Widgets/Materials/MI_UI_Pilgrim_StarComplete_GradientBurst_Gold_Transulcent
    backgroundLayer1=/PilgrimCore/Widgets/Materials/MI_Pilgrim_LightRays_Star_Gold
    backgroundLayer2=/PilgrimCore/Widgets/Materials/MI_UI_Pilgrim_StarComplete_Flare_Gold_LessBright
    particles=/PilgrimCore/Widgets/Materials/MI_UI_Piligrim_SquareParticles_Left_Orange
    triangles=/Game/UI/Foundation/Materials/MI_FX_TriangleScroll_Orange_Strong_Inst
    textOutlineColor=<rgba 0.666 0.1665 0 0.75>
    textShadowColor=<rgba 0.2 0 0 0.5>
   }
   FriendScore {
    backgroundLayer0=/PilgrimCore/Widgets/Materials/MI_UI_Pilgrim_StarComplete_GradientBurst_Purple_Transulcent
    backgroundLayer1=/PilgrimCore/Widgets/Materials/MI_Pilgrim_LightRays_Star
    backgroundLayer2=&MI_UI_Pilgrim_StarComplete_GradientBurst_Purple_Transulcent
    particles=/PilgrimCore/Widgets/Materials/MI_UI_Piligrim_SquareParticles_Left
    triangles=/Game/UI/Foundation/Materials/MI_FX_TriangleScroll_Purple_Strong
    textOutlineColor=<rgba 0.6308 0 0.6654 0.749>
    textShadowColor=<rgba 0.1248 0 0.2016 0.502>
   }
   HighScore {
    backgroundLayer0=&MI_UI_Pilgrim_StarComplete_GradientBurst_Gold_Transulcent
    backgroundLayer1=&MI_Pilgrim_LightRays_Star_Gold
    backgroundLayer2=&MI_UI_Pilgrim_StarComplete_Flare_Gold_LessBright
    particles=&MI_UI_Piligrim_SquareParticles_Left_Orange
    triangles=&MI_FX_TriangleScroll_Orange_Strong_Inst
    textOutlineColor=<rgba 0.666 0.1665 0 0.75>
    textShadowColor=<rgba 0.2 0 0 0.5>
   }
  }}
 NotificationText {type=string, default="GOOD THING!"}
 NotificationType {type=E_PilgrimRewardNotifType, default=Default}
]
widgets {
 info {inheritedWidgetCount=0, parentClass=&CommonUserWidget, rootWidgetClass=&Overlay, widgetCount=17}
 widgets [
  Overlay_0 {widgetClassPath=&Overlay, widget=Overlay_0}
  Overlay_Title {
   widgetClassPath=&Overlay
   parent=Overlay_0
   widget=Overlay_Title
   slot {class=OverlaySlot, properties {horizontalAlignment=HAlign_Fill, verticalAlignment=VAlign_Fill, padding=<ltrb 0 -8 0 0>}}
  }
  Overlay-AccoladeBG {
   widgetClassPath=&Overlay
   bIsVariable=true
   parent=Overlay_Title
   widget=Overlay-AccoladeBG
   slot {class=OverlaySlot, properties {horizontalAlignment=HAlign_Fill, verticalAlignment=VAlign_Center}}
  }
  FortVisualAttachment_SizeBox {
   widgetClassPath=&FortVisualAttachment
   bIsVariable=true
   parent=Overlay-AccoladeBG
   widget=FortVisualAttachment_SizeBox
   properties {bOverride_HeightOverride=true, bOverride_WidthOverride=true, heightOverride=44, widthOverride=640, contentAnchor=<0.5 0.5>}
   slot {class=OverlaySlot, properties {horizontalAlignment=HAlign_Center, verticalAlignment=VAlign_Center}}
  }
  Overlay_2 {
   widgetClassPath=&Overlay
   parent=FortVisualAttachment_SizeBox
   widget=Overlay_2
   properties {renderTransform {angle=0, scale=<1.125 1>, shear=<0 0>, translation=<0 0>}}
   slot {class=SizeBoxSlot}
  }
  Shadow {
   widgetClassPath=&Image
   parent=Overlay_2
   widget=Shadow
   properties {
    renderOpacity=0.333
    visibility=SelfHitTestInvisible
    brush {
     bIsDynamicallyLoaded=false
     drawAs=Image
     imageType=NoImage
     mirroring=NoMirror
     resourceName=None
     resourceObject=/PilgrimCore/Widgets/Materials/MI_UI_Pilgrim_StarComplete_GradientBurst_Shadow
     tiling=NoTile
     imageSize=<32 32>
     margin=<ltrb 0 0 0 0>
     outlineSettings {
      bUseBrushTransparency=false
      roundingType=HalfHeightRadius
      width=0
      color {colorUseRule=UseColor_Specified, specifiedColor=<rgba 0 0 0 0>}
      cornerRadii=<0 0 0 0>
     }
     tintColor {colorUseRule=UseColor_Specified, specifiedColor=<rgba 1 1 1 1>}
     uVRegion {bIsValid=false, max=<0 0>, min=<0 0>}
    }
    renderTransform {angle=0, scale=<4 4>, shear=<0 0>, translation=<0 0>}
   }
   slot {class=OverlaySlot, properties {horizontalAlignment=HAlign_Fill, verticalAlignment=VAlign_Fill}}
  }
  Image_TitleBG_2 {
   widgetClassPath=&Image
   bIsVariable=true
   parent=Overlay_2
   widget=Image_TitleBG_2
   properties {
    visibility=SelfHitTestInvisible
    brush {
     bIsDynamicallyLoaded=false
     drawAs=Image
     imageType=NoImage
     mirroring=NoMirror
     resourceName=None
     resourceObject=&MI_UI_Pilgrim_StarComplete_Flare_Gold_LessBright
     tiling=NoTile
     imageSize=<32 32>
     margin=<ltrb 0 0 0 0>
     outlineSettings {
      bUseBrushTransparency=false
      roundingType=HalfHeightRadius
      width=0
      color {colorUseRule=UseColor_Specified, specifiedColor=<rgba 0 0 0 0>}
      cornerRadii=<0 0 0 0>
     }
     tintColor {colorUseRule=UseColor_Specified, specifiedColor=<rgba 1 1 1 1>}
     uVRegion {bIsValid=false, max=<0 0>, min=<0 0>}
    }
    renderTransform {angle=0, scale=<2 1>, shear=<0 0>, translation=<0 0>}
   }
   slot {class=OverlaySlot, properties {horizontalAlignment=HAlign_Fill, verticalAlignment=VAlign_Fill}}
  }
  Image_TitleBG_1 {
   widgetClassPath=&Image
   bIsVariable=true
   parent=Overlay_2
   widget=Image_TitleBG_1
   properties {
    visibility=SelfHitTestInvisible
    brush {
     bIsDynamicallyLoaded=false
     drawAs=Image
     imageType=NoImage
     mirroring=NoMirror
     resourceName=None
     resourceObject=&MI_Pilgrim_LightRays_Star_Gold
     tiling=NoTile
     imageSize=<32 32>
     margin=<ltrb 0 0 0 0>
     outlineSettings {
      bUseBrushTransparency=false
      roundingType=HalfHeightRadius
      width=0
      color {colorUseRule=UseColor_Specified, specifiedColor=<rgba 0 0 0 0>}
      cornerRadii=<0 0 0 0>
     }
     tintColor {colorUseRule=UseColor_Specified, specifiedColor=<rgba 1 1 1 1>}
     uVRegion {bIsValid=false, max=<0 0>, min=<0 0>}
    }
    renderTransform {angle=0, scale=<1.5 3>, shear=<0 0>, translation=<0 0>}
   }
   slot {class=OverlaySlot, properties {horizontalAlignment=HAlign_Fill, verticalAlignment=VAlign_Fill}}
  }
  HorizontalBox {
   widgetClassPath=&HorizontalBox
   parent=Overlay_2
   widget=HorizontalBox
   properties {renderTransform {angle=0, scale=<1.25 1.125>, shear=<0 0>, translation=<0 0>}}
   slot {class=OverlaySlot, properties {horizontalAlignment=HAlign_Fill, verticalAlignment=VAlign_Center, padding=<ltrb 32 0 32 0>}}
  }
  ImageTriangleL_1 {
   widgetClassPath=&Image
   bIsVariable=true
   parent=HorizontalBox
   widget=ImageTriangleL_1
   properties {
    visibility=SelfHitTestInvisible
    brush {
     bIsDynamicallyLoaded=false
     drawAs=Image
     imageType=NoImage
     mirroring=NoMirror
     resourceName=None
     resourceObject=&MI_FX_TriangleScroll_Orange_Strong_Inst
     tiling=NoTile
     imageSize=<144 32>
     margin=<ltrb 0 0 0 0>
     outlineSettings {
      bUseBrushTransparency=false
      roundingType=HalfHeightRadius
      width=0
      color {colorUseRule=UseColor_Specified, specifiedColor=<rgba 0 0 0 0>}
      cornerRadii=<0 0 0 0>
     }
     tintColor {colorUseRule=UseColor_Specified, specifiedColor=<rgba 1 1 1 1>}
     uVRegion {bIsValid=false, max=<0 0>, min=<0 0>}
    }
    renderTransform {angle=0, scale=<-1 1>, shear=<0 0>, translation=<0 0>}
    renderTransformPivot=<0.5 1>
   }
   slot {class=HorizontalBoxSlot, properties {verticalAlignment=VAlign_Center, size {sizeRule=Fill, value=0.5}}}
  }
  ImageTriangleR_1 {
   widgetClassPath=&Image
   bIsVariable=true
   parent=HorizontalBox
   widget=ImageTriangleR_1
   properties {
    visibility=SelfHitTestInvisible
    brush {
     bIsDynamicallyLoaded=false
     drawAs=Image
     imageType=NoImage
     mirroring=NoMirror
     resourceName=None
     resourceObject=&MI_FX_TriangleScroll_Orange_Strong_Inst
     tiling=NoTile
     imageSize=<144 32>
     margin=<ltrb 0 0 0 0>
     outlineSettings {
      bUseBrushTransparency=false
      roundingType=HalfHeightRadius
      width=0
      color {colorUseRule=UseColor_Specified, specifiedColor=<rgba 0 0 0 0>}
      cornerRadii=<0 0 0 0>
     }
     tintColor {colorUseRule=UseColor_Specified, specifiedColor=<rgba 1 1 1 1>}
     uVRegion {bIsValid=false, max=<0 0>, min=<0 0>}
    }
    renderTransformPivot=<0.5 1>
   }
   slot {class=HorizontalBoxSlot, properties {verticalAlignment=VAlign_Center, size {sizeRule=Fill, value=0.5}}}
  }
  Image_TitleBG {
   widgetClassPath=&Image
   bIsVariable=true
   parent=Overlay_2
   widget=Image_TitleBG
   properties {
    renderOpacity=0.75
    visibility=SelfHitTestInvisible
    brush {
     bIsDynamicallyLoaded=false
     drawAs=Image
     imageType=NoImage
     mirroring=NoMirror
     resourceName=None
     resourceObject=&MI_UI_Pilgrim_StarComplete_GradientBurst_Gold_Transulcent
     tiling=NoTile
     imageSize=<32 32>
     margin=<ltrb 0 0 0 0>
     outlineSettings {
      bUseBrushTransparency=false
      roundingType=HalfHeightRadius
      width=0
      color {colorUseRule=UseColor_Specified, specifiedColor=<rgba 0 0 0 0>}
      cornerRadii=<0 0 0 0>
     }
     tintColor {colorUseRule=UseColor_Specified, specifiedColor=<rgba 1 1 1 1>}
     uVRegion {bIsValid=false, max=<0 0>, min=<0 0>}
    }
    renderTransform {angle=0, scale=<3 1>, shear=<0 0>, translation=<0 0>}
   }
   slot {class=OverlaySlot, properties {horizontalAlignment=HAlign_Fill, verticalAlignment=VAlign_Fill}}
  }
  HorizontalBox_0 {
   widgetClassPath=&HorizontalBox
   parent=Overlay_2
   widget=HorizontalBox_0
   properties {renderTransform {angle=0, scale=<1.75 1>, shear=<0 0>, translation=<0 0>}}
   slot {class=OverlaySlot, properties {horizontalAlignment=HAlign_Fill, verticalAlignment=VAlign_Center}}
  }
  ImageParticlesL {
   widgetClassPath=&Image
   bIsVariable=true
   parent=HorizontalBox_0
   widget=ImageParticlesL
   properties {
    visibility=SelfHitTestInvisible
    brush {
     bIsDynamicallyLoaded=false
     drawAs=Image
     imageType=NoImage
     mirroring=NoMirror
     resourceName=None
     resourceObject=&MI_UI_Piligrim_SquareParticles_Left_Orange
     tiling=NoTile
     imageSize=<144 32>
     margin=<ltrb 0 0 0 0>
     outlineSettings {
      bUseBrushTransparency=false
      roundingType=HalfHeightRadius
      width=0
      color {colorUseRule=UseColor_Specified, specifiedColor=<rgba 0 0 0 0>}
      cornerRadii=<0 0 0 0>
     }
     tintColor {colorUseRule=UseColor_Specified, specifiedColor=<rgba 1 1 1 1>}
     uVRegion {bIsValid=false, max=<0 0>, min=<0 0>}
    }
   }
   slot {class=HorizontalBoxSlot, properties {verticalAlignment=VAlign_Center, size {sizeRule=Fill, value=0.5}}}
  }
  ImageParticlesR {
   widgetClassPath=&Image
   bIsVariable=true
   parent=HorizontalBox_0
   widget=ImageParticlesR
   properties {
    visibility=SelfHitTestInvisible
    brush {
     bIsDynamicallyLoaded=false
     drawAs=Image
     imageType=NoImage
     mirroring=NoMirror
     resourceName=None
     resourceObject=&MI_UI_Piligrim_SquareParticles_Left_Orange
     tiling=NoTile
     imageSize=<144 32>
     margin=<ltrb 0 0 0 0>
     outlineSettings {
      bUseBrushTransparency=false
      roundingType=HalfHeightRadius
      width=0
      color {colorUseRule=UseColor_Specified, specifiedColor=<rgba 0 0 0 0>}
      cornerRadii=<0 0 0 0>
     }
     tintColor {colorUseRule=UseColor_Specified, specifiedColor=<rgba 1 1 1 1>}
     uVRegion {bIsValid=false, max=<0 0>, min=<0 0>}
    }
    renderTransform {angle=0, scale=<-1 -1>, shear=<0 0>, translation=<0 0>}
   }
   slot {class=HorizontalBoxSlot, properties {verticalAlignment=VAlign_Center, size {sizeRule=Fill, value=0.5}}}
  }
  SizeBox_0 {
   widgetClassPath=&SizeBox
   parent=Overlay_Title
   widget=SizeBox_0
   properties {bOverride_MaxDesiredWidth=true, maxDesiredWidth=1000}
   slot {class=OverlaySlot, properties {horizontalAlignment=HAlign_Center, padding=<ltrb 0 6 0 0>}}
  }
  Text_Title {
   widgetClassPath=&CommonTextBlock
   bIsVariable=true
   parent=SizeBox_0
   widget=Text_Title
   properties {
    clipping=ClipToBounds
    justification=Center
    mobileFontSizeMultiplier=1.05
    text="GOOD THING!"
    textTransformPolicy=ToUpper
    font {
     bForceMonospaced=false
     bMaterialIsStencil=false
     fontMaterial=null
     fontObject=&HeadingNow
     letterSpacing=0
     monospacedWidth=1
     size=16.5
     skewAmount=0
     typefaceFontName=86 Bold Italic
     outlineSettings {
      bApplyOutlineToDropShadows=false
      bMiteredCorners=false
      bSeparateFillAlpha=false
      outlineMaterial=null
      outlineSize=2
      outlineColor=<rgba 0.666 0.1665 0 0.75>
     }
    }
    shadowColorAndOpacity=<rgba 0.2 0 0 0.5>
    shadowOffset=<4 5>
   }
   slot {class=SizeBoxSlot}
  }
 ]
}
